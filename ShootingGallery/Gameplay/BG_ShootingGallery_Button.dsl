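; Interactable shooting gallery button that starts the game through its BPShootingGalleryGameManager reference.
; BeginPlay hides the mesh and blocks interaction when the ShootingGalleryDisabled scalable float is true.
; A one-second interaction clears AllowInteraction and passes the pawn to StartGame behind a DoOnce, which the ResetOnce event also connects to.
; PlaySpark activates the sparks effect once, only on non-authority instances.
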(event EventBeginPlay
  (if (ScalableFloat|GetValueasBool (Variables|Default|GetShootingGalleryDisabled))
    (Variables|Default|SetAllowInteraction false)
    (Rendering|SetVisibility (Variables|Default|GetButtonMesh) false false)
    (Variables|Interaction|SetbAllowInteract false)))

(event Interaction|EventBlueprintOnInteract
  (params (InteractingPawn :type Fort%20Pawn%20Object%20Reference) (InteractionBeingAttempted :type EInteractionBeingAttempted%20Enum :default "FirstInteraction"))
  (Variables|Default|SetAllowInteraction false)
  (Utilities|FlowControl|DoOnce false
    (:Completed
      (Class|BPShootingGalleryGameManager|StartGame (Variables|Default|GetManagerRef) InteractingPawn))))

(event Custom|ResetOnce
  (Utilities|FlowControl|DoOnce false
    (:Completed
      (Class|BPShootingGalleryGameManager|StartGame (Variables|Default|GetManagerRef) (stale Interaction|EventBlueprintOnInteract InteractingPawn)))))

(event Custom|PlaySpark
  (Networking|SwitchHasAuthority
    (:Remote
      (Utilities|FlowControl|DoOnce false
        (:Completed
          (Components|Activation|Activate (Variables|Default|GetSparks_Effect) false))))))

(fn BlueprintCanInteract
  (params (InteractingPawn :type Fort%20Pawn%20Object%20Reference) (InteractionBeingAttempted :type EInteractionBeingAttempted%20Enum :default "FirstInteraction") (InteractionType :type TInteractionType%20Enum :default "IT_NoInteraction"))
  (return (Variables|Default|GetAllowInteraction)))

(fn BlueprintGetInteractionString
  (params (InteractingPawn :type Fort%20Pawn%20Object%20Reference) (InteractionBeingAttempted :type EInteractionBeingAttempted%20Enum :default "FirstInteraction"))
  (return (Variables|Default|GetInteractionName)))

(fn BlueprintGetInteractionTime
  (params (InteractingPawn :type Fort%20Pawn%20Object%20Reference) (InteractionBeingAttempted :type EInteractionBeingAttempted%20Enum :default "FirstInteraction"))
  (return 1.0 true))

(fn ConstructionScript)
