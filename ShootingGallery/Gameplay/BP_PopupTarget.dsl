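; Pop-up shooting gallery target that registers hits only in one state of its enum-driven cycle, then rotates its mesh and reports the score.
; The authority makes it invulnerable, binds damage and advances states, using a timeline to lerp mesh rotation between lowered and deployed angles.
; A valid hit starts the reverse pop-up, changes the manager's score and broadcasts a custom verb message for the damage causer's owner.
; Remote instances play the target-down, impact and delayed score sounds.
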(event EventBeginPlay
  ;[1 1 1] Init
  (Networking|SwitchHasAuthority
    (:Authority
      (Variables|BuildingActor|SetbIsInvulnerable true)
      (EventDispatchers|BindEventtoOnDamaged self (delegate Custom|Hit))
      (Variables|Default|SetCurrentTargetState "NewEnumerator0")
      (Variables|Default|SetStartLoweredRelativeRotation "95.000000,0.000000,0.000000")
      (Variables|Default|SetEndDeployedRelativeRotation "0, 0, 0"))))

(event Custom|Hit
  (params (DamagedActor :type Actor%20Object%20Reference) (Damage :type Float%20%28single-precision%29) (InstigatedBy :type Controller%20Object%20Reference) (DamageCauser :type Actor%20Object%20Reference) (HitLocation :type Vector) (FHitComponent :type Primitive%20Component%20Object%20Reference) (BoneName :type Name) (Momentum :type Vector))
  ;[1 1 1] Init
  (switch Utilities|FlowControl|Switch|SwitchonE_ShootingGalleryTargetState (Variables|Default|GetCurrentTargetState)
    (:NewEnumerator2
      (CallFunction|TargetHit)
      (CustomVerb|BroadcastCustomVerbMessage (CustomVerb|MakeCustomVerbMessage (Actor|GetOwner DamageCauser) "(TagName=\"CustomVerb.ThemePark.ShootingGallery.TargetHit.Good\")")))))

(event Custom|InitTargetState
  (params (TargetState :type E_ShootingGalleryTargetState%20Enum :default "NewEnumerator0"))
  ;[1 1 1] Set State
  (Networking|SwitchHasAuthority
    (:Authority
      (bind _output_get (Variables|Default|SetCurrentTargetState TargetState))
      (switch Utilities|FlowControl|Switch|SwitchonE_ShootingGalleryTargetState _output_get
        (:NewEnumerator1
          (CallFunction|UpdateTargetStateAfterDeploy))
        (:NewEnumerator4
          (CallFunction|PopupTargetReverse))))))

(event Custom|UpdateTargetStateAfterDeploy
  ;[1 1 1] Deploy Target
  (switch Utilities|FlowControl|Switch|SwitchonE_ShootingGalleryTargetState (Variables|Default|GetCurrentTargetState)
    (:NewEnumerator1
      (Variables|Default|SetCurrentTargetState "NewEnumerator2"))
    (:NewEnumerator2
      (Variables|Default|SetCurrentTargetState "NewEnumerator4"))
    (:NewEnumerator4
      (Variables|Default|SetCurrentTargetState "NewEnumerator0"))))

(event Custom|PopupTargetReverse
  (Networking|SwitchHasAuthority
    (:Authority
      (bind (_direction _float) (Timeline|Timeline PlayFromStart)
        (:Update
          (Transformation|SetRelativeRotation (Variables|Default|GetTargetMesh_0) (Math|Rotator|Lerp(Rotator) (Variables|Default|GetStartLoweredRelativeRotation) (Variables|Default|GetEndDeployedRelativeRotation) _float true) false false))
        (:Finished
          (CallFunction|UpdateTargetStateAfterDeploy))))
    (:Remote
      (Audio|SpawnSoundatLocation "/ShootingGallery/Audio/SC_ShootingGallery_TargetDown.SC_ShootingGallery_TargetDown" (Transformation|GetWorldLocation (Variables|Default|GetTargetMesh_0)))
      (Timeline|Timeline PlayFromStart))))

(event Custom|TargetHit
  ;[1 1 1] Target Hit
  (CallFunction|InitTargetState "NewEnumerator4")
  (Networking|SwitchHasAuthority
    (:Authority
      (Class|BPShootingGalleryGameManager|ChangeScore (Variables|Default|GetManagerRef) true))
    (:Remote
      (Audio|SpawnSoundatLocation "/Game/Sounds/Creative/Toys/ShootingTargets/Target_Impact_Bullseye_Cue.Target_Impact_Bullseye_Cue" (Transformation|GetWorldLocation (Variables|Default|GetTargetMesh_0)))
      (Utilities|FlowControl|Delay 0.33)
      (Audio|SpawnSoundatLocation "/Game/Athena/Sounds/Quests/Clays_Score_Hit_Cue.Clays_Score_Hit_Cue" (Transformation|GetWorldLocation (Variables|Default|GetTargetMesh_0))))))

(fn ConstructionScript)
