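; Game manager for a Fortnite shooting gallery minigame that runs timed rounds of pop-up and sliding targets, tracking score and awarding loot.
; On authority, StartGame shuffles deploy delays, cycles random pop-up targets via timeline, starts sliders and runs the countdown.
; Game over fires a loot cannon using team-size score tiers, then resets up to round limits.
; OnRep functions update text and material displays; remote clients trigger Music Party music.
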(event EventBeginPlay
  (bind _musicpartycontrollerclass (Variables|Default|GetMusicPartyControllerClass))
  ;[1 1 1] Init
  (bind (_gamestate _playlist) (Class|FortAsyncActionCurrentPlaylistReady|PlaylistDataReadyAsync)
    (:OnReady
      (Variables|Default|SetPlaylist _playlist)
      (Networking|SwitchHasAuthority
        (:Authority
          (bind _durationseconds (CallFunction|GetGameDuration (Variables|Default|GetAllDeployedDurationDelays)))
          (Variables|Default|SetFullGameDuration _durationseconds)
          (Variables|Default|SetCurrentGameDuration (Math|Conversions|ToFloat(Integer) (Variables|Default|GetFullGameDuration)))
          (Rendering|Components|TextRender|SetText (Variables|Default|GetTimerTextRender) (Utilities|Text|ToText(Float) (Variables|Default|GetCurrentGameDuration) "HalfToEven" false true 1 324 0 0))
          (if (ScalableFloat|GetValueasBool (Variables|Default|GetDisablePlayerCollsionForTargets))
            (for _array_element (Variables|Default|GetAllPopupTargetsRef)
              (Collision|SetCollisionResponsetoChannel (Variables|/ShootingGallery/Gameplay/BP_PopupTarget.BP_PopupTarget_C|GetTargetMesh_0 _array_element) "ECC_Pawn" "ECR_Ignore"))
            (for _array_element_2 (Variables|Default|GetAllSliderTargetRefs)
              (Collision|SetCollisionResponsetoChannel (Variables|/ShootingGallery/Gameplay/BP_SlidingTarget.BP_SlidingTarget_C|GetTargetMesh_0 _array_element_2) "ECC_Pawn" "ECR_Ignore"))))
        (:Remote
          (if (Utilities|IsValidClass _musicpartycontrollerclass)
            (bind _returnvalue (Actor|GetActorOfClass _musicpartycontrollerclass))
            (Variables|Default|SetMusicPartyController _returnvalue)))))))

(event EventTick
  (params (DeltaSeconds :type Float%20%28single-precision%29)))

(event Custom|StartGame
  (params (InteractingPawn :type Fort%20Pawn%20Object%20Reference))
  ;[1 1 1] Start Game
  (Variables|Default|SetInteracting%20Pawn InteractingPawn)
  (Networking|SwitchHasAuthority
    (:Authority
      (if (ScalableFloat|GetValueasBool (Variables|Default|GetDisableShootingGallery))
        _
        (elif (< (Variables|Default|GetNumberOfRoundsPlayed) (ScalableFloat|GetValueasInteger (Variables|Default|GetMaximumNumberOfRounds)))
          (Utilities|FlowControl|Sequence
            (:then_0
              (Utilities|FlowControl|Delay 2.0)
              (Utilities|Array|Shuffle (Variables|Default|GetAllDeployedDurationDelays))
              (Utilities|FlowControl|Sequence
                (:then_0
                  (CallFunction|DeployTargets))
                (:then_1
                  (for _array_element (Variables|Default|GetAllSliderTargetRefs)
                    (Class|BPSlidingTarget|InitTargetState _array_element "NewEnumerator1")
                    (Variables|/ShootingGallery/Gameplay/BP_SlidingTarget.BP_SlidingTarget_C|SetIsTargetMoving true _array_element)))
                (:then_2
                  (bind _durationseconds (CallFunction|GetGameDuration (Variables|Default|GetAllDeployedDurationDelays)))
                  (Variables|Default|SetFullGameDuration _durationseconds)
                  (Variables|Default|SetCurrentGameDuration (Math|Conversions|ToFloat(Integer) _durationseconds))
                  (Rendering|Components|TextRender|SetText (Variables|Default|GetTimerTextRender) (Utilities|Text|ToText(Float) (Variables|Default|GetCurrentGameDuration) "HalfToEven" false true 1 324 0 0))
                  (CallFunction|CountDownTimer))
                (:then_3
                  (CallFunction|PlayGameMusic))))
            (:then_1
              (CallFunction|PlayGameStartStingerMusic))))))))

(event Custom|SetActorTick
  (params (Enabled :type Boolean))
  ;[1 1 1] Start Game
  (Actor|Tick|SetActorTickEnabled Enabled))

(event Custom|DeployTargets
  ;[1 1 1] Move Targets
  (bind _selectedrandomtargets (CallFunction|PickRandomTargets))
  (Variables|Default|SetCurrentRandomTargets _selectedrandomtargets)
  (for _array_element (Variables|Default|GetCurrentRandomTargets)
    (switch Utilities|FlowControl|Switch|SwitchonE_ShootingGalleryTargetState (Variables|/ShootingGallery/Gameplay/BP_PopupTarget.BP_PopupTarget_C|GetCurrentTargetState _array_element)
      (:NewEnumerator0
        (Variables|/ShootingGallery/Gameplay/BP_PopupTarget.BP_PopupTarget_C|SetCurrentTargetState "NewEnumerator1" _array_element)
        (CallFunction|PopupTargetStart))))
  (Utilities|FlowControl|Delay (Utilities|Array|Get(aref) (Variables|Default|GetAllDeployedDurationDelays) (Variables|Default|GetCurrentDelayInt)))
  (for _array_element_2 (Variables|Default|GetCurrentRandomTargets)
    (switch Utilities|FlowControl|Switch|SwitchonE_ShootingGalleryTargetState (Variables|/ShootingGallery/Gameplay/BP_PopupTarget.BP_PopupTarget_C|GetCurrentTargetState _array_element_2)
      (:NewEnumerator2
        (Class|BPPopupTarget|InitTargetState _array_element_2 "NewEnumerator4"))))
  (bind _result (Math|Integer|IncrementInt (Variables|Default|GetCurrentDelayInt))
    (:%20%20
      (if (>= _result (Utilities|Array|Length (Variables|Default|GetAllDeployedDurationDelays)))
        (Variables|Default|SetbGameOver true)
        (CallFunction|PlayGameEndStingerMusic)
        (CallFunction|GiveRewards)
        (Development|PrintString "Shooting Gallery: Game over" true true "(R=0.000000,G=0.660000,B=1.000000,A=1.000000)" 1e+01)
        (else
          (Utilities|FlowControl|Delay 1.5)
          (CallFunction|DeployTargets))))))

(event Custom|PopupTargetStart
  ;[1 1 1] Popup Target Timeline
  (bind (_direction _float) (Timeline|PopupTargetTimeline PlayFromStart)
    (:Update
      (for _array_element (Variables|Default|GetCurrentRandomTargets)
        (Transformation|SetRelativeRotation (Variables|/ShootingGallery/Gameplay/BP_PopupTarget.BP_PopupTarget_C|GetTargetMesh_0 _array_element) (Math|Rotator|Lerp(Rotator) (Variables|/ShootingGallery/Gameplay/BP_PopupTarget.BP_PopupTarget_C|GetStartLoweredRelativeRotation _array_element) (Variables|/ShootingGallery/Gameplay/BP_PopupTarget.BP_PopupTarget_C|GetEndDeployedRelativeRotation _array_element) _float true) false false)
        (Networking|SwitchHasAuthority
          (:Remote
            (Audio|SpawnSoundatLocation "/ShootingGallery/Audio/SC_ShootingGallery_TargetUp.SC_ShootingGallery_TargetUp" (Transformation|GetWorldLocation (Variables|/ShootingGallery/Gameplay/BP_PopupTarget.BP_PopupTarget_C|GetTargetMesh_0 _array_element)))))))
    (:Finished
      (for _array_element_2 (Variables|Default|GetCurrentRandomTargets)
        (Class|BPPopupTarget|UpdateTargetStateAfterDeploy _array_element_2)))))

(event Custom|SlideTargetStart
  ;[1 1 1] Sliding Target Move Timeline
  (Audio|Components|Audio|Play (Variables|Default|GetSC_ShootingGallery_EngineOn) 0.0)
  (bind (_direction _float) (Timeline|SlideTargetTimeline PlayFromStart)
    (:Update
      (for _array_element (Variables|Default|GetAllSliderTargetRefs)
        (Variables|/ShootingGallery/Gameplay/BP_SlidingTarget.BP_SlidingTarget_C|SetIsTargetMoving true _array_element)
        (Transformation|SetRelativeLocation (Variables|/ShootingGallery/Gameplay/BP_SlidingTarget.BP_SlidingTarget_C|GetTargetMesh_0 _array_element) (Math|Vector|Lerp(Vector) (Variables|/ShootingGallery/Gameplay/BP_SlidingTarget.BP_SlidingTarget_C|GetRelativeStartLocation _array_element) (Variables|/ShootingGallery/Gameplay/BP_SlidingTarget.BP_SlidingTarget_C|GetRelativeEndLocation _array_element) _float) false false)))
    (:Finished
      (for _array_element_2 (Variables|Default|GetAllSliderTargetRefs)
        (Class|BPSlidingTarget|FinishedSlidingTargetMove _array_element_2))
      (if (Variables|Default|GetbGameOver)
        (CallFunction|EventStopSlidingTargetEngine)))))

(event Custom|SlideTargetReverseFromEnd
  ;[1 1 1] Sliding Target Move Timeline
  (Timeline|SlideTargetTimeline ReverseFromEnd))

(event Custom|CountDownTimer
  ;[1 1 1] Countdown Clock
  (bind _returnvalue (Utilities|Time|SetTimerbyEvent (delegate Custom|DecrementTime) 1.0 true))
  (Variables|Default|SetCountdownClockHandle _returnvalue))

(event Custom|Event_Stop%20Sliding%20Target%20Engine
  ;[1 1 1] Sliding Target Move Timeline
  (Networking|CanExecuteCosmeticEvents
    (:True
      (Audio|PlaySoundatLocation "/ShootingGallery/Audio/SC_ShootingGallery_Stop.SC_ShootingGallery_Stop" (Transformation|GetWorldLocation (Variables|Default|GetSC_ShootingGallery_EngineOn)))
      (Audio|Components|Audio|FadeOut (Variables|Default|GetSC_ShootingGallery_EngineOn) 1.0 0.0 "Linear" :!class /Script/Engine.AudioComponent))))

(event Custom|DecrementTime
  ;[1 1 1] Countdown Clock
  (bind _result (Math|Integer|DecrementInt (Variables|Default|GetFullGameDuration))
    (:%20%20
      (bind _output_get (Variables|Default|SetCurrentGameDuration (Math|Conversions|ToFloat(Integer) (Math|Integer|Clamp(Integer) _result 0 999))))
      (Rendering|Components|TextRender|SetText (Variables|Default|GetTimerTextRender) (Utilities|Text|ToText(Float) _output_get "HalfToEven" false true 1 324 0 0)))))

(event Custom|ChangeScore
  (params (Add :type Boolean :default true))
  (bind _currentgamescore (Variables|Default|GetCurrentGameScore))
  (bind _returnvalue (Utilities|String|ToString(Integer) _currentgamescore))
  ;[1 1 1] Add Score
  (if Add
    (bind _result (Math|Integer|IncrementInt _currentgamescore)
      (:%20%20
        (Variables|Default|SetScoreText (Utilities|Text|ToText(Integer) (Math|Integer|Clamp(Integer) _result 0 9999)))
        (Variables|Default|SetCurrentGameScore (Math|Integer|Clamp(Integer) _result 0 9999))
        (Development|PrintString (Utilities|String|Append "Score = " _returnvalue) true true "(R=0.000000,G=0.660000,B=1.000000,A=1.000000)" 5.0)))
    (else
      (bind _result_2 (Math|Integer|DecrementInt _currentgamescore)
        (:%20%20
          (Variables|Default|SetScoreText (Utilities|Text|ToText(Integer) (Math|Integer|Clamp(Integer) _result_2 0 9999)))
          (Variables|Default|SetCurrentGameScore (Math|Integer|Clamp(Integer) _result_2 0 9999))
          (Development|PrintString (Utilities|String|Append "Score = " _returnvalue) true true "(R=0.000000,G=0.660000,B=1.000000,A=1.000000)" 5.0))))))

(event Custom|GiveRewards
  (bind _lootcannonref (Variables|Default|GetLootCannonRef))
  ;[1 1 1] Rewards
  (Utilities|FlowControl|Delay 0.2)
  (if (> (Variables|Default|GetCurrentGameScore) 0)
    (bind _tiertouse (CallFunction|LookUpReward (Variables|Default|GetCurrentGameScore)))
    (Variables|Default|SetLootTierReward _tiertouse)
    (Utilities|IsValid _lootcannonref
      (:Is%20Valid
        (Class|BGALootCannon|FireLootCannon _lootcannonref (Variables|Default|GetLootTierReward) (Variables|Default|GetInteracting%20Pawn)))
      (:Is%20Not%20Valid
        _))
    (Utilities|FlowControl|Sequence
      (:then_0
        (Utilities|FlowControl|Delay 6.0)
        (CallFunction|ResetGame))
      (:then_1
        (Utilities|FlowControl|Delay 1.0)
        (for _array_element (Variables|Default|GetAllSliderTargetRefs)
          (Class|BPSlidingTarget|ResetTargettoStart _array_element))
        (CallFunction|StopGameMusic))
      (:then_2
        (Math|Integer|IncrementInt (Variables|Default|GetNumberOfRoundsPlayed))))
    (else
      _)))

(event Custom|ResetGame
  (bind _buttonref (Variables|Default|GetButtonRef))
  ;[1 1 1] Reset Game
  (Variables|Default|SetCurrentDelayInt 0)
  (Variables|Default|SetCurrentGameScore 0)
  (Variables|Default|SetbGameOver false)
  (for _array_element (Variables|Default|GetAllPopupTargetsRef)
    (Class|BPPopupTarget|InitTargetState _array_element "NewEnumerator0"))
  (if (< (Variables|Default|GetNumberOfRoundsPlayed) (ScalableFloat|GetValueasInteger (Variables|Default|GetMaximumNumberOfRounds)))
    (bind _result (Math|Integer|IncrementInt (Variables|Default|GetCurrentNumOfGames))
      (:%20%20
        (if (> _result (Variables|Default|GetNumOfAllowedGames))
          (Class|BGShootingGalleryButton|PlaySpark (Variables|Default|GetButtonRef))
          (else
            (Class|BGShootingGalleryButton|ResetOnce _buttonref)
            (Variables|/ShootingGallery/Gameplay/BG_ShootingGallery_Button.BG_ShootingGallery_Button_C|SetAllowInteraction true _buttonref)
            (Variables|Default|SetScoreText "NSLOCTEXT(\"[5C91089D2801E60485EB77F93A234C91]\", \"4F81B8344FEA55FB32101BBECF8F1431\", \"0\")")
            (Utilities|Time|ClearandInvalidateTimerbyHandle (Variables|Default|GetCountdownClockHandle))))))))

(event Custom|PlayGameMusic
  ;[1 1 1] Game Music
  (Networking|SwitchHasAuthority
    (:Remote
      (bind _musicpartycontroller (Variables|Default|GetMusicPartyController)
        (:then
          (Class|MusicPartyControllerParameterInterface|SetTriggerParameter(Message) _musicpartycontroller "StartGame"))
        (:else)))))

(event Custom|StopGameMusic
  ;[1 1 1] Game Music
  (Networking|SwitchHasAuthority
    (:Remote
      (bind _musicpartycontroller (Variables|Default|GetMusicPartyController)
        (:then
          (Class|MusicPartyControllerParameterInterface|SetTriggerParameter(Message) _musicpartycontroller "StopGame"))
        (:else)))))

(event Custom|PlayGameStartStingerMusic
  ;[1 1 1] Game Music
  (Networking|SwitchHasAuthority))

(event Custom|PlayGameEndStingerMusic
  ;[1 1 1] Game Music
  (Networking|SwitchHasAuthority))

(fn GetGameDuration
  (params (DeployDurationDelays :type Array%20of%20Float%20%28double-precision%29s))
  (for _array_element DeployDurationDelays
    (LocalVariables|SetCurrentCount (+ (LocalVariables|GetCurrentCount) _array_element)))
  (return (+ (Math|Float|Truncate (LocalVariables|GetCurrentCount)) 5)))

(fn LookUpReward
  (params (Point :type Integer))
  (bind _playlist (Variables|Default|GetPlaylist))
  (bind _maxteamsize (Variables|/Script/FortniteGame.FortPlaylist|GetMaxTeamSize _playlist))
  (bind _rewardmapsquad (Variables|Default|GetRewardMapSquad))
  (bind _rewardmaptrio (Variables|Default|GetRewardMapTrio))
  (bind _rewardmapduo (Variables|Default|GetRewardMapDuo))
  (bind _rewardmapsolo (Variables|Default|GetRewardMapSolo))
  (bind _returnvalue (Utilities|Select :Index _maxteamsize :Option%200 _rewardmapsolo :Option%201 _rewardmapsolo :Option%202 _rewardmapduo :Option%203 _rewardmaptrio :Option%204 _rewardmapsquad :Option%205 _rewardmapsquad))
  (bind _keys (Utilities|Map|Keys _returnvalue))
  (bind _array_element (Utilities|Array|ForEachLoopwithBreak _keys)
    (:LoopBody
      (if (< (Variables|GetPoint) _array_element)
        _)
      (LocalVariables|SetLastBestLootTier (Utilities|Map|Find _returnvalue _array_element)))
    (:Completed
      _))
  (return (LocalVariables|GetLastBestLootTier)))

(fn MoveSliderTarget
  (params (TickDelta :type Float%20%28single-precision%29))
  (bind _currentsliderref (LocalVariables|GetCurrentSliderRef))
  (bind _currentsliderref_2 (LocalVariables|GetCurrentSliderRef))
  (bind _currentsliderref_3 (LocalVariables|GetCurrentSliderRef))
  (bind _targetmesh_0 (Variables|/ShootingGallery/Gameplay/BP_SlidingTarget.BP_SlidingTarget_C|GetTargetMesh_0 _currentsliderref_3))
  (for _array_element (Variables|Default|GetAllSliderTargetRefs)
    (bind _output_get (LocalVariables|SetCurrentSliderRef _array_element))
    (Utilities|IsValid _output_get
      (:Is%20Valid
        (if (Variables|/ShootingGallery/Gameplay/BP_SlidingTarget.BP_SlidingTarget_C|GetIsTargetMoving _currentsliderref)
          (if (Utilities|Enum|Equal(Enum) (Variables|/ShootingGallery/Gameplay/BP_SlidingTarget.BP_SlidingTarget_C|GetCurrentTargetState _currentsliderref) "NewEnumerator2")
            (Transformation|SetRelativeLocation _targetmesh_0 (Math|Interpolation|VInterptoConstant (Variables|/Script/Engine.SceneComponent|GetRelativeLocation _targetmesh_0) (Variables|/ShootingGallery/Gameplay/BP_SlidingTarget.BP_SlidingTarget_C|GetRelativeStartLocation _currentsliderref_3) (Variables|GetTickDelta) 1.5e+02) false false)
            (if (Utilities|Operators|Equal(==) (Variables|/ShootingGallery/Gameplay/BP_SlidingTarget.BP_SlidingTarget_C|GetRelativeStartLocation (LocalVariables|GetCurrentSliderRef)) (Variables|/Script/Engine.SceneComponent|GetRelativeLocation _targetmesh_0) 0.1)
              (if (Variables|/ShootingGallery/Gameplay/BP_SlidingTarget.BP_SlidingTarget_C|GetbResetTarget _currentsliderref_2)
                (Class|BPSlidingTarget|InitTargetState _currentsliderref_2 "NewEnumerator0")
                (else
                  (Class|BPSlidingTarget|InitTargetState _currentsliderref_2 "NewEnumerator1"))))
            (else
              (Transformation|SetRelativeLocation (Variables|/ShootingGallery/Gameplay/BP_SlidingTarget.BP_SlidingTarget_C|GetTargetMesh_0 (LocalVariables|GetCurrentSliderRef)) (Math|Interpolation|VInterptoConstant (Variables|/Script/Engine.SceneComponent|GetRelativeLocation (Variables|/ShootingGallery/Gameplay/BP_SlidingTarget.BP_SlidingTarget_C|GetTargetMesh_0 (LocalVariables|GetCurrentSliderRef))) (Variables|/ShootingGallery/Gameplay/BP_SlidingTarget.BP_SlidingTarget_C|GetRelativeEndLocation (LocalVariables|GetCurrentSliderRef)) (Variables|GetTickDelta) 1.5e+02) false false)
              (if (Utilities|Operators|Equal(==) (Variables|/ShootingGallery/Gameplay/BP_SlidingTarget.BP_SlidingTarget_C|GetRelativeEndLocation (LocalVariables|GetCurrentSliderRef)) (Variables|/Script/Engine.SceneComponent|GetRelativeLocation (Variables|/ShootingGallery/Gameplay/BP_SlidingTarget.BP_SlidingTarget_C|GetTargetMesh_0 (LocalVariables|GetCurrentSliderRef))) 0.1)
                (Class|BPSlidingTarget|InitTargetState (LocalVariables|GetCurrentSliderRef) "NewEnumerator2")))))))))

(fn OnRep_CurrentGameDuration
  (Rendering|Components|TextRender|SetText (Variables|Default|GetTimerTextRender) (Utilities|Text|ToText(Float) (Variables|Default|GetCurrentGameDuration) "HalfToEven" false true 1 324 0 0))
  (Rendering|Material|SetCustomPrimitiveDataFloat (Variables|Default|GetTimerMesh) 0 (Variables|Default|GetCurrentGameDuration)))

(fn OnRep_CurrentGameScore
  (Rendering|Components|TextRender|SetText (Variables|Default|GetScoreTextRender) (Variables|Default|GetScoreText))
  (Rendering|Material|SetCustomPrimitiveDataVector3 (Variables|Default|GetScoreMesh) 0 (Math|Vector|MakeVector (Math|Conversions|ToFloat(Integer) (Variables|Default|GetCurrentGameScore)) (Utilities|Time|GetGameTimeinSeconds) (Math|Conversions|ToFloat(Boolean) (and (> (Variables|Default|GetCurrentGameScore) (Variables|Default|GetLastCurrentScore)) (!= (Variables|Default|GetCurrentGameScore) 0)))))
  (Variables|Default|SetLastCurrentScore (Variables|Default|GetCurrentGameScore)))

(fn PickRandomMusic
  (bind _possiblesongslocal (LocalVariables|GetPossibleSongsLocal))
  (Utilities|Array|Shuffle _possiblesongslocal)
  (return (Utilities|Array|Get(acopy) _possiblesongslocal)))

(fn PickRandomTargets
  (Utilities|Array|Shuffle (Variables|Default|GetAllPopupTargetsRef))
  (Utilities|Array|Clear (Variables|Default|GetCurrentRandomTargets))
  (bind _index (Utilities|FlowControl|ForLoopwithBreak 0 (Math|Integer|Clamp(Integer) (Math|Random|RandomIntegerinRange 3 6) 3 (Utilities|Array|LastIndex (Variables|Default|GetAllPopupTargetsRef))))
    (:LoopBody
      (Utilities|Array|Add (LocalVariables|GetAllSelectedLocal) (Utilities|Array|Get(acopy) (Variables|Default|GetAllPopupTargetsRef) _index)))
    (:Completed
      (return (LocalVariables|GetAllSelectedLocal)))))

(fn PopupTargetRetract
  (params (Alpha :type Float%20%28single-precision%29) (PopupTarget :type BP%20Popup%20Target%20Object%20Reference))
  (Transformation|SetRelativeRotation (Variables|/ShootingGallery/Gameplay/BP_PopupTarget.BP_PopupTarget_C|GetTargetMesh_0 (Variables|GetPopupTarget)) (Math|Rotator|Lerp(Rotator) (Variables|/ShootingGallery/Gameplay/BP_PopupTarget.BP_PopupTarget_C|GetStartLoweredRelativeRotation (Variables|GetPopupTarget)) (Variables|/ShootingGallery/Gameplay/BP_PopupTarget.BP_PopupTarget_C|GetEndDeployedRelativeRotation (Variables|GetPopupTarget)) Alpha true) false false))

(fn PopupTargetStartForAll
  (params (Alpha :type Float%20%28single-precision%29)))

(fn RotatePopupTarget
  (params (TickDelta :type Float%20%28single-precision%29))
  (bind (_relativerotation_roll _relativerotation_pitch) (Math|Rotator|BreakRotator (Variables|/Script/Engine.SceneComponent|GetRelativeRotation_Roll (Variables|/ShootingGallery/Gameplay/BP_PopupTarget.BP_PopupTarget_C|GetTargetMesh_0 (LocalVariables|GetRandomTargetRef)))))
  (bind (_startloweredrelativerotation_roll _startloweredrelativerotation_pitch) (Math|Rotator|BreakRotator (Variables|/ShootingGallery/Gameplay/BP_PopupTarget.BP_PopupTarget_C|GetStartLoweredRelativeRotation_Roll (LocalVariables|GetRandomTargetRef))))
  (bind (_relativerotation_roll_2 _relativerotation_pitch_2) (Math|Rotator|BreakRotator (Variables|/Script/Engine.SceneComponent|GetRelativeRotation_Roll (Variables|/ShootingGallery/Gameplay/BP_PopupTarget.BP_PopupTarget_C|GetTargetMesh_0 (LocalVariables|GetRandomTargetRef)))))
  (bind (_enddeployedrelativerotation_roll _enddeployedrelativerotation_pitch) (Math|Rotator|BreakRotator (Variables|/ShootingGallery/Gameplay/BP_PopupTarget.BP_PopupTarget_C|GetEndDeployedRelativeRotation_Roll (LocalVariables|GetRandomTargetRef))))
  (for _array_element (Variables|Default|GetCurrentRandomTargets)
    (bind _output_get (LocalVariables|SetRandomTargetRef _array_element))
    (Utilities|IsValid _output_get
      (:Is%20Valid
        (if true
          (switch Utilities|FlowControl|Switch|SwitchonE_ShootingGalleryTargetState (Variables|/ShootingGallery/Gameplay/BP_PopupTarget.BP_PopupTarget_C|GetCurrentTargetState (LocalVariables|GetRandomTargetRef))
            (:NewEnumerator1
              (Transformation|SetRelativeRotation (Variables|/ShootingGallery/Gameplay/BP_PopupTarget.BP_PopupTarget_C|GetTargetMesh_0 (LocalVariables|GetRandomTargetRef)) "0, 0, 0" false false)
              (if (Math|Float|NearlyEqual(Float) _enddeployedrelativerotation_pitch _relativerotation_pitch 0.1)
                (Class|BPPopupTarget|InitTargetState (LocalVariables|GetRandomTargetRef) "NewEnumerator2")))
            (:NewEnumerator4
              (Transformation|SetRelativeRotation (Variables|/ShootingGallery/Gameplay/BP_PopupTarget.BP_PopupTarget_C|GetTargetMesh_0 (LocalVariables|GetRandomTargetRef)) (Math|Interpolation|RInterptoConstant (Variables|/Script/Engine.SceneComponent|GetRelativeRotation (Variables|/ShootingGallery/Gameplay/BP_PopupTarget.BP_PopupTarget_C|GetTargetMesh_0 (LocalVariables|GetRandomTargetRef))) (Variables|/ShootingGallery/Gameplay/BP_PopupTarget.BP_PopupTarget_C|GetStartLoweredRelativeRotation (LocalVariables|GetRandomTargetRef)) (Variables|GetTickDelta) 2.7e+02) false false)
              (if (Math|Float|NearlyEqual(Float) _relativerotation_pitch_2 _startloweredrelativerotation_pitch 0.1)
                (Class|BPPopupTarget|InitTargetState (LocalVariables|GetRandomTargetRef) "NewEnumerator0")))))))))

(fn ConstructionScript)
