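; Shooting gallery sliding "bad" target that reports hits as a penalty.
; BeginPlay makes it invulnerable, binds OnDamaged on the server and computes mesh endpoints along X, optionally flipped.
; The server state machine asks the game manager to slide it, alternating direction after each move until game over; a timeline lerps it home on reset.
; Hits broadcast a Bad custom verb, then a DoOnce gates the server score change and client friendly-impact sound.
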(event EventBeginPlay
  ;[1 1 1] Init
  (Networking|SwitchHasAuthority
    (:Authority
      (Variables|BuildingActor|SetbIsInvulnerable true)
      (EventDispatchers|BindEventtoOnDamaged self (delegate Custom|Hit))
      (Variables|Default|SetCurrentTargetState "NewEnumerator0"))
    (:Remote
      _))
  (bind _output_get (Variables|Default|SetRelativeStartLocation (Variables|/Script/Engine.SceneComponent|GetRelativeLocation (Variables|Default|GetTargetMesh_0))))
  (if (Variables|Default|GetFlipDirection)
    (Variables|Default|SetRelativeEndLocation (Math|Vector|MakeVector (+ (.x _output_get) (* (Variables|Default|GetMoveDistance) -1.0)) (.y _output_get) (.z _output_get)))
    (else
      (Variables|Default|SetRelativeEndLocation (Math|Vector|MakeVector (+ (.x _output_get) (Variables|Default|GetMoveDistance)) (.y _output_get) (.z _output_get))))))

(event Custom|Hit
  (params (DamagedActor :type Actor%20Object%20Reference) (Damage :type Float%20%28single-precision%29) (InstigatedBy :type Controller%20Object%20Reference) (DamageCauser :type Actor%20Object%20Reference) (HitLocation :type Vector) (FHitComponent :type Primitive%20Component%20Object%20Reference) (BoneName :type Name) (Momentum :type Vector))
  ;[1 1 1] Init
  (switch Utilities|FlowControl|Switch|SwitchonE_ShootingGalleryTargetState (Variables|Default|GetCurrentTargetState)
    (:NewEnumerator1
      _)
    (:NewEnumerator2
      _)
    (:NewEnumerator4
      _))
  (CallFunction|TargetHit)
  (CustomVerb|BroadcastCustomVerbMessage (CustomVerb|MakeCustomVerbMessage (Actor|GetOwner DamageCauser) "(TagName=\"CustomVerb.ThemePark.ShootingGallery.TargetHit.Bad\")")))

(event Custom|InitTargetState
  (params (TargetState :type E_ShootingGalleryTargetState%20Enum :default "NewEnumerator0"))
  (bind _managerref (Variables|Default|GetManagerRef))
  ;[1 1 1] Set State
  (Networking|SwitchHasAuthority
    (:Authority
      (Variables|Default|SetCurrentTargetState TargetState)
      (switch Utilities|FlowControl|Switch|SwitchonE_ShootingGalleryTargetState (Variables|Default|GetCurrentTargetState)
        (:NewEnumerator0
          (Variables|Default|SetIsTargetMoving false)
          (Development|PrintString "Shooting Gallery: Target is idle" true true "(R=0.000000,G=0.660000,B=1.000000,A=1.000000)" 1e+01))
        (:NewEnumerator1
          (Class|BPShootingGalleryGameManager|SlideTargetStart _managerref))
        (:NewEnumerator2
          (Class|BPShootingGalleryGameManager|SlideTargetReversefromEnd _managerref))
        (:NewEnumerator4
          (Class|BPShootingGalleryGameManager|SlideTargetStart _managerref))))))

(event Custom|FinishedSlidingTargetMove
  ;[1 1 1] Target Move Finished
  (Variables|Default|SetIsTargetMoving false)
  (if (not (Variables|/ShootingGallery/Gameplay/BP_ShootingGallery_GameManager.BP_ShootingGallery_GameManager_C|GetbGameOver (Variables|Default|GetManagerRef)))
    (switch Utilities|FlowControl|Switch|SwitchonE_ShootingGalleryTargetState (Variables|Default|GetCurrentTargetState)
      (:NewEnumerator1
        (CallFunction|InitTargetState "NewEnumerator2"))
      (:NewEnumerator2
        (CallFunction|InitTargetState "NewEnumerator4"))
      (:NewEnumerator4
        (CallFunction|InitTargetState "NewEnumerator2")))))

(event Custom|ResetTargetToStart
  ;[1 1 1] Reset Location
  (Variables|Default|SetIsTargetMoving false)
  (if (Variables|Default|GetIsTargetMoving)
    (Utilities|FlowControl|Delay 1.0))
  (Variables|Default|SetCurrentRelativeLocation (Variables|/Script/Engine.SceneComponent|GetRelativeLocation (Variables|Default|GetTargetMesh_0)))
  (bind (_direction _float) (Timeline|Reset PlayFromStart)
    (:Update
      (Variables|Default|SetIsTargetMoving true)
      (Transformation|SetRelativeLocation (Variables|Default|GetTargetMesh_0) (Math|Vector|Lerp(Vector) (Variables|Default|GetCurrentRelativeLocation) (Variables|Default|GetRelativeStartLocation) _float) false false))))

(event Custom|TargetHit
  ;[1 1 1] Target Hit
  (Utilities|FlowControl|DoOnce false
    (:Completed
      (Utilities|FlowControl|Sequence
        (:then_0
          _)
        (:then_1
          (Utilities|FlowControl|Delay 0.5)))
      (Networking|SwitchHasAuthority
        (:Authority
          (Class|BPShootingGalleryGameManager|ChangeScore (Variables|Default|GetManagerRef) false))
        (:Remote
          (Audio|SpawnSoundatLocation :Sound "/ShootingGallery/Audio/SC_ShootingGallery_Target_Impact_Friendly.SC_ShootingGallery_Target_Impact_Friendly" :Location (Transformation|GetWorldLocation (Variables|Default|GetTargetMesh_0)) :VolumeMultiplier 1.8 :PitchMultiplier 1.0 :StartTime 0.0 :bAutoDestroy true))))))

(fn ConstructionScript)
